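# Translucent, unlit, two-sided material for Niagara mesh particles, used by the fan mesh effects of the StallionSmoke glider.
# Particle color drives emissive through MF_FX_EmissiveAdjustments.
# Opacity multiplies particle alpha by a product of fades, overlay-blended with panning turbulence noise, then depth-faded.
# Named reroutes define a UV-based tip fade, a view-angle Fresnel fade and a tip push along vertex normals.
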Material /BRCosmetics/Gadgets/Assets/VinderTech_GliderChute/Glider_StallionSmoke/FX/Materials/M_Glider_StallionSmoke_FanMeshes
graph {
 properties {
  bUsedWithNiagaraMeshParticles=true
  blendMode=BLEND_Translucent
  pixelDepthOffsetMode=PDOM_Legacy
  shadingModel=MSM_Unlit
  twoSided=true
  usedShadingModels=Unlit
 }
 nodes [
  n1 = ParticleColor()
  n2 = Multiply(A=NamedRerouteUsage(), B=NamedRerouteUsage())
  n3 = TextureCoordinate()
  n4 = VectorParameter() {parameterName=Noise Scale and Speed, defaultValue=<rgba 1 1 999 0>}
  n5 = AppendVector(A=@n4.R, B=@n4.R)
  n6 = TextureSample(UVs=Panner(Coordinate=Multiply(A=@n3, B=@n5), Speed=AppendVector(A=@n4.B, B=@n4.B)) {bFractionalPart=true}) {samplerType=SAMPLERTYPE_LinearColor, texture=/Game/Effects/Fort_Effects/Effects/Environment/Desert/Dust/Organize/T_Turbulence_Seamless}
  DepthFade(
   Opacity=Multiply(
    A=@n1.A
    B=MaterialFunctionCall(Base=@n2, Blend=@n6.R) {materialFunction=&Blend_Overlay}.Result
   )
   FadeDistance=ScalarParameter() {defaultValue=22, parameterName=Depth Fade Distance}
  )
  n7 = ScalarParameter() {defaultValue=2, parameterName=Exponent}
  MaterialFunctionCall(Axis=Transform(None=Constant3Vector() {constant=<rgba 1 0 0 0>}) {transformSourceType=TRANSFORMSOURCE_Instance}, Exponent=@n7) {materialFunction=&AxisAlignedFresnel}
  MaterialFunctionCall(Emissive=@n1.RGB) {materialFunction=&MF_FX_EmissiveAdjustments}
  n8 = TextureCoordinate()
  NamedRerouteDeclaration(None=Power(Base=ComponentMask(None=@n8) {g=true}, Exp=ScalarParameter() {defaultValue=1, parameterName=Tip Fade Exponent})) {name=Tip Fade, nodeColor=<rgba 1 0 0.6797 1>}
  n9 = DotProduct(A=VertexNormalWS(), B=CameraVectorWS())
  NamedRerouteDeclaration(None=Power(Base=Abs(None=@n9), Exp=@n7)) {name=Fresnel Fade, nodeColor=<rgba 1 0.7266 0 1>}
  n10 = OneMinus(None=ComponentMask(None=TextureCoordinate()) {g=true})
  n11 = ScalarParameter() {defaultValue=4, parameterName=Push Distance}
  NamedRerouteDeclaration(None=Multiply(A=Multiply(A=@n10, B=@n11), B=VertexNormalWS())) {name=Tip Push, nodeColor=<rgba 0 1 0.8125 1>}
  NamedRerouteUsage()
 ]
}
