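# Additive, Default Lit glass material for the StallionSmoke glider, usable on skeletal meshes despite its MI_ prefix.
# An HDR sky texture sampled along the world-space reflection vector is scaled by Cubemap_Brightness, then combined with a pale blue Color parameter by both add and multiply.
# The multiplied result feeds MF_FX_EmissiveAdjustments.
# An Opacity parameter is added to a Fresnel term with adjustable exponent and base reflection.
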Material /BRCosmetics/Gadgets/Assets/VinderTech_GliderChute/Glider_StallionSmoke/Materials/MI_Glider_StallionSmoke_Glass
graph {
 properties {
  bUsedWithMeshDeformer=true
  bUsedWithSkeletalMesh=true
  blendMode=BLEND_Additive
  pixelDepthOffsetMode=PDOM_Legacy
  usedShadingModels=Default Lit
 }
 nodes [
  n1 = Multiply(
   A=TextureSample(UVs=ReflectionVectorWS()) {texture=/Game/Packages/Fortress_Sky/TexturesHDR/T_hdrsky_test03_low}.RGB
   B=ScalarParameter() {defaultValue=1, parameterName=Cubemap_Brightness}
  )
  n2 = VectorParameter() {parameterName=Color, defaultValue=<rgba 0.6137 0.8641 0.9427 1>}
  Add(A=@n1, B=@n2.RGB)
  Add(
   A=ScalarParameter() {parameterName=Opacity}
   B=Fresnel(
    ExponentIn=ScalarParameter() {defaultValue=5, parameterName=Fresnel_Exp}
    BaseReflectFractionIn=ScalarParameter() {defaultValue=0.04, parameterName=Reflection}
   )
  )
  MaterialFunctionCall(Emissive=Multiply(A=@n1, B=@n2.RGB)) {materialFunction=&MF_FX_EmissiveAdjustments}
 ]
}
