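# Physics asset for the StallionSmoke glider cosmetic mesh, part of the VinderTech glider chute gadget assets.
# It holds a single body on the Glider_StallionSmoke_MESH_LOD0 bone, shaped as one large rotated capsule offset from the origin.
# No constraints are defined, so it acts as one simple collision volume with no articulated ragdoll.
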PhysicsAsset /BRCosmetics/Gadgets/Assets/VinderTech_GliderChute/Glider_StallionSmoke/Meshes/StallionSmoke_Glider_PhysicsAsset
bodies {Glider_StallionSmoke_MESH_LOD0 [{shapeType=Capsule, length=157.1521, radius=132.7277, center=<-0.007 67.5467 5.6589>, rotation=<rot 6.6879 89.9947 -90>}]}
constraints []
